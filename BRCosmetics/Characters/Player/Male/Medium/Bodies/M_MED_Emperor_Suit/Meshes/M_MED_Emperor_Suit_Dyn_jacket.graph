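# Physics asset for the dynamic jacket flaps on the medium male Emperor Suit body, covering front and back flaps on left and right.
# Each flap is a two-capsule chain with limited swing and zero-degree twist limits, hung from a tiny MSTR capsule.
# The MSTR bodies and both thigh capsules connect to the pelvis capsule with free angular constraints, giving the flaps lower-body volumes alongside them.
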PhysicsAsset /BRCosmetics/Characters/Player/Male/Medium/Bodies/M_MED_Emperor_Suit/Meshes/M_MED_Emperor_Suit_Dyn_jacket
bodies {
 dyn_jacket_flap_bk_l_01 [{shapeType=Capsule, length=7.4103, radius=6, center=<3.7504 0.0004 6>, rotation=<rot 90 0 0>}]
 dyn_jacket_flap_bk_l_01_MSTR [{shapeType=Capsule, length=0.5, radius=0.5, rotation=<rot 90 0 0>}]
 dyn_jacket_flap_bk_l_02 [{shapeType=Capsule, length=7.4103, radius=6, center=<3.7504 0.0004 6>, rotation=<rot 90 0 0>}]
 dyn_jacket_flap_bk_r_01 [{shapeType=Capsule, length=7.4103, radius=6, center=<3.7504 0.0004 6>, rotation=<rot 90 0 0>}]
 dyn_jacket_flap_bk_r_01_MSTR [{shapeType=Capsule, length=0.5, radius=0.5, rotation=<rot 90 0 0>}]
 dyn_jacket_flap_bk_r_02 [{shapeType=Capsule, length=7.4103, radius=6, center=<3.7504 0.0004 6>, rotation=<rot 90 0 0>}]
 dyn_jacket_flap_l_01 [{shapeType=Capsule, length=7.4103, radius=6, center=<3.7504 0.0004 6>, rotation=<rot 90 0 0>}]
 dyn_jacket_flap_l_01_MSTR [{shapeType=Capsule, length=0.5, radius=0.5, rotation=<rot 90 0 0>}]
 dyn_jacket_flap_l_02 [{shapeType=Capsule, length=7.4103, radius=6, center=<3.7504 0.0004 6>, rotation=<rot 90 0 0>}]
 dyn_jacket_flap_r_01 [{shapeType=Capsule, length=7.4103, radius=6, center=<3.7504 0.0004 6>, rotation=<rot 90 0 0>}]
 dyn_jacket_flap_r_01_MSTR [{shapeType=Capsule, length=0.5, radius=0.5, rotation=<rot 90 0 0>}]
 dyn_jacket_flap_r_02 [{shapeType=Capsule, length=7.4103, radius=6, center=<3.7504 0.0004 6>, rotation=<rot 90 0 0>}]
 pelvis [{shapeType=Capsule, length=4.9867, radius=14.5989, center=<-3.9253 1.7292 0>}]
 thigh_l [{shapeType=Capsule, length=18.2367, radius=9.5677, center=<-16.1179 -0.3176 1.0848>, rotation=<rot 90 0 0>}]
 thigh_r [{shapeType=Capsule, length=18.2367, radius=9.5677, center=<16.1179 0.3176 -1.0848>, rotation=<rot 90 0 180>}]
}
constraints [
 {
  bone1Name=dyn_jacket_flap_bk_r_01_MSTR
  bone2Name=pelvis
  swing1LimitDegrees=45
  swing1Motion=Free
  swing2LimitDegrees=45
  swing2Motion=Free
  twistLimitDegrees=45
  twistMotion=Free
 }
 {bone1Name=dyn_jacket_flap_bk_r_01, bone2Name=dyn_jacket_flap_bk_r_01_MSTR, swing1LimitDegrees=10, swing2LimitDegrees=45, twistLimitDegrees=0}
 {bone1Name=dyn_jacket_flap_bk_r_02, bone2Name=dyn_jacket_flap_bk_r_01, swing1LimitDegrees=10, swing2LimitDegrees=45, twistLimitDegrees=0}
 {
  bone1Name=dyn_jacket_flap_r_01_MSTR
  bone2Name=pelvis
  swing1LimitDegrees=45
  swing1Motion=Free
  swing2LimitDegrees=45
  swing2Motion=Free
  twistLimitDegrees=45
  twistMotion=Free
 }
 {bone1Name=dyn_jacket_flap_r_01, bone2Name=dyn_jacket_flap_r_01_MSTR, swing1LimitDegrees=10, swing2LimitDegrees=45, twistLimitDegrees=0}
 {bone1Name=dyn_jacket_flap_r_02, bone2Name=dyn_jacket_flap_r_01, swing1LimitDegrees=10, swing2LimitDegrees=45, twistLimitDegrees=0}
 {
  bone1Name=dyn_jacket_flap_l_01_MSTR
  bone2Name=pelvis
  swing1LimitDegrees=45
  swing1Motion=Free
  swing2LimitDegrees=45
  swing2Motion=Free
  twistLimitDegrees=45
  twistMotion=Free
 }
 {bone1Name=dyn_jacket_flap_l_01, bone2Name=dyn_jacket_flap_l_01_MSTR, swing1LimitDegrees=10, swing2LimitDegrees=45, twistLimitDegrees=0}
 {bone1Name=dyn_jacket_flap_l_02, bone2Name=dyn_jacket_flap_l_01, swing1LimitDegrees=10, swing2LimitDegrees=45, twistLimitDegrees=0}
 {
  bone1Name=dyn_jacket_flap_bk_l_01_MSTR
  bone2Name=pelvis
  swing1LimitDegrees=45
  swing1Motion=Free
  swing2LimitDegrees=45
  swing2Motion=Free
  twistLimitDegrees=45
  twistMotion=Free
 }
 {bone1Name=dyn_jacket_flap_bk_l_01, bone2Name=dyn_jacket_flap_bk_l_01_MSTR, swing1LimitDegrees=10, swing2LimitDegrees=45, twistLimitDegrees=0}
 {bone1Name=dyn_jacket_flap_bk_l_02, bone2Name=dyn_jacket_flap_bk_l_01, swing1LimitDegrees=10, swing2LimitDegrees=45, twistLimitDegrees=0}
 {
  bone1Name=thigh_l
  bone2Name=pelvis
  swing1LimitDegrees=45
  swing1Motion=Free
  swing2LimitDegrees=45
  swing2Motion=Free
  twistLimitDegrees=45
  twistMotion=Free
 }
 {
  bone1Name=thigh_r
  bone2Name=pelvis
  swing1LimitDegrees=45
  swing1Motion=Free
  swing2LimitDegrees=45
  swing2Motion=Free
  twistLimitDegrees=45
  twistMotion=Free
 }
]
