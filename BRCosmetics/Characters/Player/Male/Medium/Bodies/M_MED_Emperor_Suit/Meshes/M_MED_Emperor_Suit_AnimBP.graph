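# Post-process-style cosmetic AnimBP for the male medium Emperor suit, copying the attached parent's pose and adding garment deformation and secondary motion.
# While the Emperor emote montage plays, EmoteAlpha blends shirt, jacket, tie and nametag joints, fading out the nametag's spine constraint.
# Two Control Rigs, thigh-yaw-driven flap offsets, nametag AnimDynamics and a gravity-overridden RigidBody jacket simulation follow.
# Separate jacket-flap AnimDynamics and thigh-constraint nodes are left unconnected.
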AnimBlueprint /BRCosmetics/Characters/Player/Male/Medium/Bodies/M_MED_Emperor_Suit/Meshes/M_MED_Emperor_Suit_AnimBP
target_skeleton=/Game/Characters/Player/Male/Medium/Bodies/M_MED_Emperor_Suit/Meshes/M_MED_Emperor_Suit_Skeleton
graphs [
 AnimGraph {nodes [
   n1 = Anim Dynamics - Bone: dyn_jacket_flap_r_01(Alpha=1, ExternalForce="0,0,0") {
    angularDampingOverride=1
    angularSpringConstant=20000
    bAngularSpring=true
    bGravityOverrideInSimSpace=true
    bOverrideAngularDamping=true
    bUseGravityOverride=true
    gravityScale=0.35
    simulationSpace=RootRelative
    alphaBoolBlend {blendOption=Linear}
    boundBone {boneName=dyn_jacket_flap_r_01}
    componentLinearAccScale=<1 1 1>
    componentLinearVelScale=<1 1 1>
    gravityOverride {z=-980}
    physicsBodyDefinitions [{
      sphereCollisionRadius=0
      boundBone {boneName=dyn_jacket_flap_r_01}
      boxExtents=<20 30 30>
      constraintSetup {angularLimitsMax=<2 2 15>, angularLimitsMin=<-2 -15 -15>, angularTarget=<1 0 0>}
      localJointOffset=<4 0 0>
     }]
   }
   n2 = Anim Dynamics - Bone: dyn_jacket_flap_r_02(ComponentPose=@n1, Alpha=1, ExternalForce="0,0,0") {
    angularDampingOverride=1
    angularSpringConstant=20000
    bAngularSpring=true
    bGravityOverrideInSimSpace=true
    bOverrideAngularDamping=true
    bUseGravityOverride=true
    gravityScale=0.35
    simulationSpace=RootRelative
    alphaBoolBlend {blendOption=Linear}
    boundBone {boneName=dyn_jacket_flap_r_02}
    componentLinearAccScale=<1 1 1>
    componentLinearVelScale=<1 1 1>
    gravityOverride {z=-980}
    physicsBodyDefinitions [{
      sphereCollisionRadius=0
      boundBone {boneName=dyn_jacket_flap_r_02}
      boxExtents=<20 30 30>
      constraintSetup {angularLimitsMax=<2 2 15>, angularLimitsMin=<-2 -15 -15>, angularTarget=<1 0 0>}
      localJointOffset=<4 0 0>
     }]
   }
   n3 = Anim Dynamics - Bone: dyn_jacket_flap_bk_r_01(ComponentPose=@n2, Alpha=1, ExternalForce="0,0,0") {
    angularDampingOverride=1
    angularSpringConstant=20000
    bAngularSpring=true
    bGravityOverrideInSimSpace=true
    bOverrideAngularDamping=true
    bUseGravityOverride=true
    gravityScale=0.35
    simulationSpace=RootRelative
    alphaBoolBlend {blendOption=Linear}
    boundBone {boneName=dyn_jacket_flap_bk_r_01}
    componentLinearAccScale=<1 1 1>
    componentLinearVelScale=<1 1 1>
    gravityOverride {z=-980}
    physicsBodyDefinitions [{
      sphereCollisionRadius=0
      boundBone {boneName=dyn_jacket_flap_bk_r_01}
      boxExtents=<20 30 30>
      constraintSetup {angularLimitsMax=<2 2 15>, angularLimitsMin=<-2 -15 -15>, angularTarget=<1 0 0>}
      localJointOffset=<4 0 0>
     }]
   }
   n4 = Anim Dynamics - Bone: dyn_jacket_flap_bk_r_02(ComponentPose=@n3, Alpha=1, ExternalForce="0,0,0") {
    angularDampingOverride=1
    angularSpringConstant=20000
    bAngularSpring=true
    bGravityOverrideInSimSpace=true
    bOverrideAngularDamping=true
    bUseGravityOverride=true
    gravityScale=0.35
    simulationSpace=RootRelative
    alphaBoolBlend {blendOption=Linear}
    boundBone {boneName=dyn_jacket_flap_bk_r_02}
    componentLinearAccScale=<1 1 1>
    componentLinearVelScale=<1 1 1>
    gravityOverride {z=-980}
    physicsBodyDefinitions [{
      sphereCollisionRadius=0
      boundBone {boneName=dyn_jacket_flap_bk_r_02}
      boxExtents=<20 30 30>
      constraintSetup {angularLimitsMax=<2 2 15>, angularLimitsMin=<-2 -15 -15>, angularTarget=<1 0 0>}
      localJointOffset=<4 0 0>
     }]
   }
   n5 = Anim Dynamics - Bone: dyn_jacket_flap_l_01(ComponentPose=@n4, Alpha=1, ExternalForce="0,0,0") {
    angularDampingOverride=1
    angularSpringConstant=20000
    bAngularSpring=true
    bGravityOverrideInSimSpace=true
    bOverrideAngularDamping=true
    bUseGravityOverride=true
    gravityScale=0.35
    simulationSpace=RootRelative
    alphaBoolBlend {blendOption=Linear}
    boundBone {boneName=dyn_jacket_flap_l_01}
    componentLinearAccScale=<1 1 1>
    componentLinearVelScale=<1 1 1>
    gravityOverride {z=-980}
    physicsBodyDefinitions [{
      sphereCollisionRadius=0
      boundBone {boneName=dyn_jacket_flap_l_01}
      boxExtents=<20 30 30>
      constraintSetup {angularLimitsMax=<2 2 15>, angularLimitsMin=<-2 -15 -15>, angularTarget=<1 0 0>}
      localJointOffset=<4 0 0>
     }]
   }
   n6 = Anim Dynamics - Bone: dyn_jacket_flap_l_02(ComponentPose=@n5, Alpha=1, ExternalForce="0,0,0") {
    angularDampingOverride=1
    angularSpringConstant=20000
    bAngularSpring=true
    bGravityOverrideInSimSpace=true
    bOverrideAngularDamping=true
    bUseGravityOverride=true
    gravityScale=0.35
    simulationSpace=RootRelative
    alphaBoolBlend {blendOption=Linear}
    boundBone {boneName=dyn_jacket_flap_l_02}
    componentLinearAccScale=<1 1 1>
    componentLinearVelScale=<1 1 1>
    gravityOverride {z=-980}
    physicsBodyDefinitions [{
      sphereCollisionRadius=0
      boundBone {boneName=dyn_jacket_flap_l_02}
      boxExtents=<20 30 30>
      constraintSetup {angularLimitsMax=<2 2 15>, angularLimitsMin=<-2 -15 -15>, angularTarget=<1 0 0>}
      localJointOffset=<4 0 0>
     }]
   }
   n7 = Anim Dynamics - Bone: dyn_jacket_flap_bk_l_01(ComponentPose=@n6, Alpha=1, ExternalForce="0,0,0") {
    angularDampingOverride=1
    angularSpringConstant=20000
    bAngularSpring=true
    bGravityOverrideInSimSpace=true
    bOverrideAngularDamping=true
    bUseGravityOverride=true
    gravityScale=0.35
    simulationSpace=RootRelative
    alphaBoolBlend {blendOption=Linear}
    boundBone {boneName=dyn_jacket_flap_bk_l_01}
    componentLinearAccScale=<1 1 1>
    componentLinearVelScale=<1 1 1>
    gravityOverride {z=-980}
    physicsBodyDefinitions [{
      sphereCollisionRadius=0
      boundBone {boneName=dyn_jacket_flap_bk_l_01}
      boxExtents=<20 30 30>
      constraintSetup {angularLimitsMax=<2 2 15>, angularLimitsMin=<-2 -15 -15>, angularTarget=<1 0 0>}
      localJointOffset=<4 0 0>
     }]
   }
   Anim Dynamics - Bone: dyn_jacket_flap_bk_l_02(ComponentPose=@n7, Alpha=1, ExternalForce="0,0,0") {
    angularDampingOverride=1
    angularSpringConstant=20000
    bAngularSpring=true
    bGravityOverrideInSimSpace=true
    bOverrideAngularDamping=true
    bUseGravityOverride=true
    gravityScale=0.35
    simulationSpace=RootRelative
    alphaBoolBlend {blendOption=Linear}
    boundBone {boneName=dyn_jacket_flap_bk_l_02}
    componentLinearAccScale=<1 1 1>
    componentLinearVelScale=<1 1 1>
    gravityOverride {z=-980}
    physicsBodyDefinitions [{
      sphereCollisionRadius=0
      boundBone {boneName=dyn_jacket_flap_bk_l_02}
      boxExtents=<20 30 30>
      constraintSetup {angularLimitsMax=<2 2 15>, angularLimitsMin=<-2 -15 -15>, angularTarget=<1 0 0>}
      localJointOffset=<4 0 0>
     }]
   }
   "Constraint - dyn_jacket_flap_fr_l_01_MSTR (R)"(Alpha=1, ConstraintWeights_0=1) {
    alphaBoolBlend {blendOption=Linear}
    boneToModify {boneName=dyn_jacket_flap_fr_l_01_MSTR}
    constraintSetup [{transformType=Rotation, targetBone {boneName=thigh_twist_01_l}}]
    constraintWeights [1]
   }
   n8 = Layered blend per bone(
    BasePose=Copy Pose From Mesh() {bUseAttachedParent=true}
    BlendPoses_0=Sequence Player 'Emote_Emperor_Into_ShirtJoints'() {sequence=/BRCosmetics/Animation/Game/MainPlayer/Emotes/Emperor/Emote_Emperor_Into_ShirtJoints}
    BlendWeights_0=Get EmoteAlpha() -> (EmoteAlpha:double)
   ) {layerSetup [{branchFilters [
       {boneName=dyn_jacket_open_top_l_01_MSTR}
       {boneName=dyn_jacket_open_top_r_01_MSTR}
       {boneName=dyn_tie_01_MSTR}
       {boneName=dyn_shirt_open_top_l_01_MSTR}
       {boneName=dyn_shirt_open_top_r_01_MSTR}
       {boneName=dyn_jacket_open_bott_01_MSTR}
       {boneName=dyn_nametag_clip_MSTR}
      ]}]}
   n9 = Control Rig(Source=@n8) {
    alphaBoolBlend {blendOption=Linear}
    controlRigAssetReference {blueprintRigClass=&/M_MED_Emperor_Suit_Deform_CtrlRig_twists.M_MED_Emperor_Suit_Deform_CtrlRig_twists_C}
   }
   n10 = Control Rig!(
    Source=@n9
    Alpha
    KneeRot_fr_l_114deg:Vector
    KneeTrans_fr_l_114deg:Vector="0.258,-3.591,0.009"
    KneeScale_fr_l_114deg:Vector="1,1,1"
    KneeScale_fr_l_130deg:Vector="1,1,1"
    KneeRot_fr_l_130deg:Vector="0,0,-51.2"
    KneeTrans_fr_l_130deg:Vector="0.258,-3.591,0.009"
   ) -> (Pose) {
    alphaBoolBlend {blendOption=Linear}
    controlRigAssetReference {blueprintRigClass=&/M_MED_Emperor_Suit_Deform_CtrlRig_sculpters.M_MED_Emperor_Suit_Deform_CtrlRig_sculpters_C}
   }
   n11 = "Constraint - dyn_nametag_clip_MSTR (PPPPP)"!(
    ComponentPose=Local To Component(LocalPose=@n10)
    Alpha=Get EmoteInvAlpha() -> (EmoteInvAlpha:double)
    ConstraintWeights_0=0.6613
    ConstraintWeights_1:float=0.2114
    ConstraintWeights_2:float=0.0613
    ConstraintWeights_3:float=0.0364
    ConstraintWeights_4:float=0.0295
   ) -> (Pose) {
    alphaBoolBlend {blendOption=Linear}
    boneToModify {boneName=dyn_nametag_clip_MSTR}
    constraintSetup [
     {targetBone {boneName=spine_04}}
     {targetBone {boneName=spine_05}}
     {targetBone {boneName=spine_03}}
     {targetBone {boneName=spine_02}}
     {targetBone {boneName=deform_pectoral_fr_l}}
    ]
    constraintWeights [1,1,1,1,1]
   }
   n12 = "Transform (Modify) Bone - Bone: dyn_jacket_flap_ft_l_MSTR"(ComponentPose=@n11, Alpha=Get thigh_rot_l(), Translation="-17,3,-3", Rotation="0,20,0", Scale="1,1,1") {
    rotationMode=BMM_Additive
    rotationSpace=BCS_BoneSpace
    scaleMode=BMM_Additive
    scaleSpace=BCS_BoneSpace
    translationMode=BMM_Additive
    translationSpace=BCS_BoneSpace
    alphaBoolBlend {blendOption=Linear}
    alphaScaleBiasClamp {bMapRange=true, inRange {max=0, min=-100}, outRange {max=0, min=1}}
    boneToModify {boneName=dyn_jacket_flap_ft_l_MSTR}
    rotation {yaw=20}
    translation=<-17 3 -3>
   }
   n13 = "Transform (Modify) Bone - Bone: dyn_jacket_flap_ft_r_MSTR"(ComponentPose=@n12, Alpha=Get thigh_rot_r(), Translation="15,-3,3", Rotation="0,20,0", Scale="1,1,1") {
    rotationMode=BMM_Additive
    rotationSpace=BCS_BoneSpace
    scaleMode=BMM_Additive
    scaleSpace=BCS_BoneSpace
    translationMode=BMM_Additive
    translationSpace=BCS_BoneSpace
    alphaBoolBlend {blendOption=Linear}
    alphaScaleBiasClamp {bMapRange=true, inRange {max=0, min=-100}, outRange {max=0, min=1}}
    boneToModify {boneName=dyn_jacket_flap_ft_r_MSTR}
    rotation {yaw=20}
    translation=<15 -3 3>
   }
   n14 = Anim Dynamics - Bone: dyn_nametag_clip(ComponentPose=@n13, Alpha=1, ExternalForce="0,0,0") {
    angularDampingOverride=1
    angularSpringConstant=20000
    bAngularSpring=true
    bGravityOverrideInSimSpace=true
    bOverrideAngularDamping=true
    bUseGravityOverride=true
    gravityScale=0.35
    simulationSpace=RootRelative
    alphaBoolBlend {blendOption=Linear}
    boundBone {boneName=dyn_nametag_clip}
    componentLinearAccScale=<1 1 1>
    componentLinearVelScale=<1 1 1>
    gravityOverride {z=-980}
    physicsBodyDefinitions [{
      sphereCollisionRadius=0
      boundBone {boneName=dyn_nametag_clip}
      boxExtents=<20 10 10>
      constraintSetup {angularLimitsMax=<0 0 3>, angularLimitsMin=<0 -15 -3>, angularTarget=<1 0 0>}
      localJointOffset=<4 0 0>
     }]
   }
   n15 = Anim Dynamics - Bone: dyn_nametag(ComponentPose=@n14, Alpha=1, ExternalForce="0,0,0") {
    angularDampingOverride=1
    angularSpringConstant=20000
    bAngularSpring=true
    bGravityOverrideInSimSpace=true
    bOverrideAngularDamping=true
    bUseGravityOverride=true
    gravityScale=0.35
    simulationSpace=RootRelative
    alphaBoolBlend {blendOption=Linear}
    boundBone {boneName=dyn_nametag}
    componentLinearAccScale=<1 1 1>
    componentLinearVelScale=<1 1 1>
    gravityOverride {z=-980}
    physicsBodyDefinitions [{
      sphereCollisionRadius=0
      boundBone {boneName=dyn_nametag}
      boxExtents=<10 15 2>
      constraintSetup {angularLimitsMax=<0 0 0>, angularLimitsMin=<0 -15 0>, angularTarget=<1 0 0>}
      localJointOffset=<4 0 0>
     }]
   }
   n16 = RigidBody!(ComponentPose=@n15, Alpha, OverrideWorldGravity:Vector=Get GravityOverride(), ExternalForce) -> (Pose) {
    bClampLinearTranslationLimitToRefPose=true
    bOverrideWorldGravity=true
    overridePhysicsAsset=&/M_MED_Emperor_Suit_Dyn_jacket
    simulationSpace=BaseBoneSpace
    alphaBoolBlend {blendOption=Linear}
    baseBoneRef {boneName=pelvis}
    componentAppliedLinearAccClamp=<3000 3000 3000>
    componentLinearAccScale {z=0.5}
    componentLinearVelScale {z=0.5}
    overrideWorldGravity {z=-980}
    simSpaceSettings {
     maxAngularAcceleration=1000
     maxAngularVelocity=3
     maxLinearAcceleration=700
     maxLinearVelocity=500
     velocityScaleZ=0
     worldAlpha=1
     externalLinearDragV=<1 1>
    }
   }
   Output Pose(Result=Component To Local(ComponentPose=@n16)) {layerGroup=None}
  ]}
 EventGraph {nodes [
   n1 = Event BlueprintUpdateAnimation()
   > n2 = Sequence!(execute) -> (then_0, then_1, then_2:Exec)
   > n3 = Gravity Override State Function DEPRECATED(CharacterPartAnimInstance=Self-Reference(), GameTimeInSeconds=GetGameTimeInSeconds(), GravityOverrideParams=Get Gravity Override Params())
   > Set GravityOverride(GravityOverride=@n3)
   n4 = Montage_IsActive!(
    self:Anim Instance Object Reference=Get Anim_Inst() -> (Anim_Inst:Anim Instance Object Reference)
    Montage:Anim Montage Object Reference=/BRCosmetics/Animation/Game/MainPlayer/Emotes/Emperor/Emote_Emperor_M
   ) -> (ReturnValue:bool)
   n5 = Set EmoteAlpha(
    execute=@n2.then_1
    EmoteAlpha:double=FInterpTo(Current=Get EmoteAlpha(), Target="To Float (Boolean)"(InBool=@n4), DeltaTime=@n1.DeltaTimeX, InterpSpeed=15)
   ) -> (Output_Get:double)
   > Set EmoteInvAlpha(EmoteInvAlpha:double=MapRangeClamped(Value=@n5, InRangeA=0, InRangeB=1, OutRangeA=1, OutRangeB=0)) -> (Output_Get:double)
   n6 = GetDeltaTransformFromRefPose(self=GetOwnerSkeletalMeshComponent(), BoneName=thigh_l, BaseName=pelvis)
   Set thigh_rot_l(execute=@n2.then_2, thigh_rot_l=BreakRotator(InRot=BreakTransform(InTransform=@n6).Rotation).Yaw)
   n7 = GetDeltaTransformFromRefPose(self=GetOwnerSkeletalMeshComponent(), BoneName=thigh_r, BaseName=pelvis)
   > Set thigh_rot_r(thigh_rot_r=BreakRotator(InRot=BreakTransform(InTransform=@n7).Rotation).Yaw)
   Event BlueprintInitializeAnimation()
   n8 = GetOwnerSkeletalMeshComponent()
   > Branch(Condition=IsValid(Object=@n8))
   > n9 = Set Skel_Mesh(Skel_Mesh=@n8)
   > Set Anim_Inst(Anim_Inst:Anim Instance Object Reference=GetAnimInstance(self=@n9)) -> (Output_Get:Anim Instance Object Reference)
  ]}
]
variables [
 Anim_Inst {type=AnimInstance, default=null}
 EmoteAlpha {type=number}
 EmoteInvAlpha {type=number, default=1}
 Gravity Override Params {type=GravityOverrideParamsStruct}
 GravityOverride {type=Vector}
 Skel_Mesh {type=SkeletalMeshComponent}
 thigh_rot_l {type=number}
 thigh_rot_r {type=number}
]
